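# Corrective Control Rig for the face accessory part of the female medium VitalPsych HoodDown cosmetic, adjusting the lowered hood as the head moves.
# A spherical pose reader measures the head bone against spine_03 within an asymmetric active region and falloff.
# Its output weights a Modify Transforms offset of rotation and translation on hood_top_Master, parent of dyn_hood_top.
# The other hair and hood bones are untouched, and pose reader debug drawing is left on.
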ControlRigBlueprint /BRCosmetics/Characters/Player/Female/Medium/Bodies/F_MED_VitalPsych_HoodDown/Meshes/Parts/F_MED_VitalPsych_HoodDown_FaceAcc_CtrlRig
graphs [
 RigVMFunctionLibrary {links [], nodes []}
 RigVMModel {nodes [
   RigUnit_BeginExecution = &RigUnit_BeginExecution()
   n1 = &RigUnit_SphericalPoseReader(
    ActiveRegionScaleFactors=(PositiveWidth=6,NegativeWidth=1.5,PositiveHeight=4,NegativeHeight=4)
    ActiveRegionSize=0.1
    Debug=(bDrawDebug=True,DebugScale=25,DebugSegments=20,DebugThickness=0.25)
    DriverAxis=(X=1,Y=0,Z=0)
    DriverItem=(Type=Bone,Name="head")
    FalloffRegionScaleFactors=(PositiveWidth=0,NegativeWidth=1,PositiveHeight=0,NegativeHeight=0)
    FalloffSize=0.35
    OptionalParentItem=(Type=Bone,Name="spine_03")
    RotationOffset=(X=90,Y=0,Z=-175)
    ExecutePin=@RigUnit_BeginExecution.ExecutePin
   )
   > &RigUnit_ModifyTransforms(
    ItemToModify=((Item=(Type=Bone,Name="hood_top_Master"),Transform=(Rotation=(X=0.0094,Y=-0.2162,Z=0.0426,W=0.9754),Translation=(X=2,Y=0,Z=2),Scale3D=(X=1,Y=1,Z=1))))
    WeightMinimum=0
    Weight=@n1.OutputParam
   )
  ]}
]
hierarchy [
 root {loc=<0 0 0>, rot=<0 0 0 1>}
 pelvis {parent=root, loc=<0 -0.5301 98.0571>, rot=<0.0127 0.707 -0.0127 -0.707>}
 spine_01 {parent=pelvis, loc=<6.1208 -0.9328 0>, rot=<0 0 0.034 -0.9994>}
 spine_02 {parent=spine_01, loc=<6.5996 -0.2095 0>, rot=<0 0 0.0238 -0.9997>}
 spine_03 {parent=spine_02, loc=<6.4914 -0.1237 0>, rot=<0 0 0.0149 -0.9999>}
 spine_04 {parent=spine_03, loc=<6.9598 0.0794 0>, rot=<0 0 -0.0132 -0.9999>}
 spine_05 {parent=spine_04, loc=<6.9987 0 0>, rot=<0 0 -0.0015 -1>}
 neck_01 {parent=spine_05, loc=<11.3138 0.1529 0>, rot=<0 0 -0.2003 -0.9797>}
 neck_02 {parent=neck_01, loc=<4.5627 0 0>, rot=<0 0 0.0504 -0.9987>}
 head {parent=neck_02, loc=<5.3914 0 0>, rot=<0 0 0.115 -0.9934>}
 dyn_L_hair_side {parent=head, loc=<5.1317 1.6113 -6.5246>, rot=<-0.1564 0.6896 0.7041 -0.0656>}
 dyn_L_hair_back {parent=head, loc=<10.5525 -6.6271 -0.9038>, rot=<-0.1371 0.6464 0.7499 -0.0332>}
 dyn_L_hair_back_down_2 {parent=dyn_L_hair_back, loc=<7.5189 -7.6894 2.8944>, rot=<0.1609 -0.1936 0.0321 -0.9673>}
 dyn_L_hair_back_down {parent=dyn_L_hair_back, loc=<7.6249 -5.1412 1.6884>, rot=<0.1616 -0.1296 0.0125 -0.9782>}
 dyn_M_hair_back_Down {parent=dyn_L_hair_back, loc=<8.5567 -0.0467 1.7761>, rot=<0.1537 -0.133 -0.0497 -0.9779>}
 dyn_R_hair_back_down {parent=dyn_L_hair_back, loc=<8.1963 6.2907 3.8684>, rot=<0.1537 -0.133 -0.0497 -0.9779>}
 dyn_hair_top {parent=head, loc=<12.9278 5.0833 1.3027>, rot=<0.4525 0.1189 0.8815 -0.0638>}
 dyn_M_hair_front {parent=head, loc=<11.8895 6.6583 -2.3926>, rot=<-0.3664 -0.6651 -0.6425 -0.103>}
 R_hair_back_mid_MASTER {parent=head, loc=<7.4631 -9.582 -0.0541>, rot=<-0.3841 0.5506 0.6823 -0.2894>}
 dyn_R_hair_back_mid {parent=R_hair_back_mid_MASTER, loc=<0.1585 0 0>, rot=<0 0 0 1>}
 M_hair_back_mid_MASTER {parent=head, loc=<7.2521 -8.7386 -3.8307>, rot=<-0.2718 0.6105 0.7038 -0.2411>}
 dyn_M_hair_back_mid {parent=M_hair_back_mid_MASTER, loc=<0.1585 0 0>, rot=<0 0 0 1>}
 MASTER_R_hair_down_1_1 {parent=neck_01, loc=<6.1352 -0.869 4.8294>, rot=<0.4692 0.529 0.5567 0.4359>}
 dyn_R_hair_down_1_2 {parent=MASTER_R_hair_down_1_1, loc=<4.3017 0 0.3119>, rot=<-0.0026 -0.0448 0.0042 -0.999>}
 dyn_R_hair_down_1_3 {parent=dyn_R_hair_down_1_2, loc=<5.6126 0 0>, rot=<-0.0047 -0.4206 0.0379 -0.9064>}
 dyn_R_hair_down_1_4 {parent=dyn_R_hair_down_1_3, loc=<8.5582 0 0>, rot=<0 0 0 1>}
 MASTER_L_hair_down_1_1 {parent=neck_01, loc=<6.1351 -0.8689 -5.1864>, rot=<-0.436 -0.5567 -0.529 -0.4692>}
 dyn_L_hair_down_1_2 {parent=MASTER_L_hair_down_1_1, loc=<4.2798 0.3563 0.3116>, rot=<-0.0019 -0.0432 -0.1048 -0.9935>}
 dyn_L_hair_down_1_3 {parent=dyn_L_hair_down_1_2, loc=<5.6519 0 0>, rot=<-0.027 -0.4213 0.077 -0.9033>}
 dyn_L_hair_down_1_4 {parent=dyn_L_hair_down_1_3, loc=<8.5651 0 0>, rot=<0 0 0 1>}
 clavicle_l {parent=spine_05, loc=<5.4831 -0.1053 -3.2946>, rot=<0.0465 -0.7289 -0.009 -0.6829>}
 upperarm_l {parent=clavicle_l, loc=<9.8658 -1.2136 -0.4105>, rot=<0.0164 -0.3245 -0.0269 -0.9454>}
 upperarm_twist_01_l {parent=upperarm_l, loc=<7.2135 -0.1134 0>, rot=<0 0 0 1>}
 clavicle_r {parent=spine_05, loc=<5.4831 -0.1053 3.2946>, rot=<0.7289 0.0465 -0.6829 0.009>}
 upperarm_r {parent=clavicle_r, loc=<-9.8658 1.2136 0.4105>, rot=<0.0164 -0.3245 -0.0269 -0.9454>}
 upperarm_twist_01_r {parent=upperarm_r, loc=<-7.2135 0.1134 0>, rot=<0 0 0 1>}
 dyn_hood {parent=spine_05, loc=<13.7592 4.5869 -0.0427>, rot=<-0.2742 0.6518 0.6552 -0.2659>}
 dyn_hood_L_ear_1 {parent=dyn_hood, loc=<14.5904 -5.2026 -3.0827>, rot=<-0.0228 0.2489 0.0761 -0.9653>}
 dyn_hood_L_ear_1_1 {parent=dyn_hood_L_ear_1, loc=<9.2267 0 0>, rot=<0 0 0 1>}
 dyn_hood_L_ear_2 {parent=dyn_hood_L_ear_1_1, loc=<7.7447 0.2309 1.2659>, rot=<-0.0168 0.1132 -0.0226 -0.9932>}
 dyn_hood_R_ear_1 {parent=dyn_hood, loc=<14.4541 5.5698 -3.0827>, rot=<0.9642 0.0883 0.2492 0.0197>}
 dyn_hood_R_ear_1_1 {parent=dyn_hood_R_ear_1, loc=<9.2267 0 0>, rot=<0 0 0 1>}
 dyn_hood_R_ear_2 {parent=dyn_hood_R_ear_1_1, loc=<7.6092 0.2252 -1.2345>, rot=<0.0168 -0.1132 -0.0226 -0.9932>}
 Hood_shoulder_fix_L {parent=spine_05, loc=<11.4852 0.4556 -7.5968>, rot=<0.0462 -0.7058 -0.0105 -0.7068>}
 Hood_shoulder_fix_R {parent=spine_05, loc=<11.105 0.2656 7.5425>, rot=<0.7058 0.0462 0.7068 -0.0105>}
 hood_top_Master {parent=spine_05, loc=<14.4482 5.1422 -0.0427>, rot=<-0.6192 0.3415 0.3683 -0.6036>}
 dyn_hood_top {parent=hood_top_Master, loc=<0 0 0>, rot=<0 0 0 1>}
 hood_top2_Master {parent=spine_05, loc=<14.0802 5.1127 -0.0427>, rot=<-0.518 0.471 0.5007 -0.509>}
 dyn_hood_top2 {parent=hood_top2_Master, loc=<11.8078 0 0>, rot=<0 0 0 1>}
]
