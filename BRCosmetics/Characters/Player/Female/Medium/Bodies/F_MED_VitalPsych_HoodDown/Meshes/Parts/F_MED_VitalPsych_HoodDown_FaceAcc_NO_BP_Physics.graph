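# Physics asset for the female VitalPsych HoodDown cosmetic's face accessory part, setting up secondary motion for hair and a lowered hood with ears.
# Capsules and spheres cover head, neck, upper spine, clavicles and upper arms, linked by free angular constraints.
# Dynamic hair uses tight swing limits off head or master bones, with all lower back strands linked to the left back bone.
# Three-bone ear chains swing more widely from a zero-limit hood body.
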PhysicsAsset /BRCosmetics/Characters/Player/Female/Medium/Bodies/F_MED_VitalPsych_HoodDown/Meshes/Parts/F_MED_VitalPsych_HoodDown_FaceAcc_NO_BP_Physics
bodies {
 M_hair_back_mid_MASTER [{shapeType=Capsule, length=0.505, radius=0.505, shapeName=M_hair_back_mid_MASTER_capsule, rotation=<rot 0 -90 0>}]
 R_hair_back_mid_MASTER [{shapeType=Capsule, length=0.505, radius=0.505, shapeName=R_hair_back_mid_MASTER_capsule, rotation=<rot 0 -90 0>}]
 clavicle_l [{shapeType=Capsule, length=9, radius=6.5, shapeName=clavicle_l_capsule, center=<4 0 -4.5>, rotation=<rot 90 0 0>}]
 clavicle_r [{shapeType=Capsule, length=9, radius=6.5, shapeName=clavicle_r_capsule, center=<-4 0 4.5>, rotation=<rot 90 0 0>}]
 dyn_L_hair_back [{shapeType=Capsule, length=2.1414, radius=1.6275, shapeName=dyn_L_hair_back_capsule, center=<2 0 0>, rotation=<rot 90 0 0>}]
 dyn_L_hair_back_down [{
   shapeType=Capsule
   length=4.3195
   radius=0.7565
   shapeName=dyn_L_hair_back_down_capsule
   center=<5.3544 0 0.0552>
   rotation=<rot 70.9318 -0.0002 -0.0001>
  }]
 dyn_L_hair_back_down_2 [{shapeType=Capsule, length=3.6224, radius=0.626, shapeName=dyn_L_hair_back_down_2_capsule, center=<3.6496 0 0>, rotation=<rot 90 0 0>}]
 dyn_L_hair_side [{shapeType=Capsule, length=2.25, radius=1.25, shapeName=dyn_L_hair_side_capsule, center=<1.705 0 0>, rotation=<rot 90 0 0>}]
 dyn_M_hair_back_Down [{
   shapeType=Capsule
   length=3.9796
   radius=2.9259
   shapeName=dyn_M_hair_back_Down_capsule
   center=<4.4456 0 0.6557>
   rotation=<rot 81.0044 -179.9966 -170.5187>
  }]
 dyn_M_hair_back_mid [{shapeType=Capsule, length=3.8834, radius=1.3513, shapeName=dyn_M_hair_back_mid_capsule, center=<3.6554 0 0>, rotation=<rot 74.8185 0 0>}]
 dyn_M_hair_front [{shapeType=Capsule, length=2, radius=2, shapeName=dyn_M_hair_front_capsule, center=<2 0 0>, rotation=<rot 90 0 0>}]
 dyn_R_hair_back_down [{
   shapeType=Capsule
   length=2.7285
   radius=1.2584
   shapeName=dyn_R_hair_back_down_capsule
   center=<3.8678 -0.0118 -0.6656>
   rotation=<rot 0 -98.6522 -70.623>
  }]
 dyn_R_hair_back_mid [{
   shapeType=Capsule
   length=3.1241
   radius=1.1187
   shapeName=dyn_R_hair_back_mid_capsule
   center=<3.1748 -0.5393 0.369>
   rotation=<rot 81.2655 -4.4877 -0.0001>
  }]
 dyn_hair_top [
  {shapeType=Capsule, length=2.75, radius=1.25, shapeName=dyn_hair_top_capsule, center=<2 0 0>, rotation=<rot 90 0 0>}
  {
   shapeType=Capsule
   length=6.0311
   radius=1.2304
   shapeName=dyn_hair_top_capsule
   center=<13.3799 1.4475 -6.2278>
   rotation=<rot -41.9925 -179.9031 168.7203>
  }
 ]
 dyn_hood [{shapeType=Capsule, length=2.0492, radius=3.4611, shapeName=dyn_hood_capsule, center=<5 0 0>, rotation=<rot 90 0 0>}]
 dyn_hood_L_ear_1 [{shapeType=Capsule, length=4.988, radius=1.5611, shapeName=dyn_hood_L_ear_1_capsule, center=<3.5 0 0>, rotation=<rot 90 0 0>}]
 dyn_hood_L_ear_1_1 [{shapeType=Capsule, length=4.8616, radius=1.5178, shapeName=dyn_hood_L_ear_1_1_capsule, center=<3 0 0>, rotation=<rot 100 0 0>}]
 dyn_hood_L_ear_2 [{shapeType=Capsule, length=4.5192, radius=1.7284, shapeName=dyn_hood_L_ear_2_capsule, center=<4 0 0>, rotation=<rot 90 0 0>}]
 dyn_hood_R_ear_1 [{shapeType=Capsule, length=3.506, radius=1.6223, shapeName=dyn_hood_R_ear_1_capsule, center=<3.5 0 0>, rotation=<rot 90 0 0>}]
 dyn_hood_R_ear_1_1 [{shapeType=Capsule, length=4.9006, radius=1.4788, shapeName=dyn_hood_R_ear_1_1_capsule, center=<3 0 0>, rotation=<rot 80 0 0>}]
 dyn_hood_R_ear_2 [{shapeType=Capsule, length=5.1335, radius=1.7224, shapeName=dyn_hood_R_ear_2_capsule, center=<4 0 0>, rotation=<rot 90 0 0>}]
 dyn_hood_top [
  {shapeType=Capsule, length=2.6171, radius=1.3632, shapeName=dyn_hood_top_capsule, center=<1 0 0>, rotation=<rot 90 0 0>}
  {
   shapeType=Capsule
   length=15.3716
   radius=1.554
   shapeName=dyn_hood_top_capsule
   center=<13.4244 -5.5524 2.6153>
   rotation=<rot -82.5863 -133.0828 -178.6365>
  }
  {
   shapeType=Capsule
   length=15.3373
   radius=1.8716
   shapeName=dyn_hood_top_capsule
   center=<12.5451 6.3453 2.8841>
   rotation=<rot -84.578 -159.3067 118.3546>
  }
  {shapeType=Capsule, length=1.5028, radius=2.4966, shapeName=dyn_hood_top_capsule, center=<17.0056 0.816 3.3765>}
 ]
 dyn_hood_top2 [{shapeType=Capsule, length=0.505, radius=0.505, shapeName=dyn_hood_top2_capsule, rotation=<rot 0 -90 0>}]
 head [{shapeType=Capsule, length=1.2431, radius=6.7888, shapeName=head_capsule, center=<5.4382 1.503 0.7141>, rotation=<rot 0 68.667 -74.8325>}]
 hood_top2_Master [{shapeType=Capsule, length=0.3047, radius=0.505, shapeName=hood_top2_Master_capsule, rotation=<rot 0 -90 0>}]
 hood_top_Master [{shapeType=Capsule, length=0.3047, radius=0.505, shapeName=hood_top_Master_capsule, rotation=<rot 0 -90 0>}]
 neck_01 [{shapeType=Capsule, length=3, radius=3.005, shapeName=neck_01_capsule, rotation=<rot 0 -90.0002 90>}]
 neck_02 [{shapeType=Capsule, length=0.505, radius=0.505, shapeName=neck_02_capsule, rotation=<rot 0 -90 0>}]
 spine_03 [{shapeType=Capsule, length=25, radius=8.5, shapeName=spine_03_capsule, center=<0 0.5871 0>, rotation=<rot 0 -90 0>}]
 spine_04 [{shapeType=Capsule, length=21.5, radius=8.255, shapeName=spine_04_capsule, center=<3.6984 1.5693 0>, rotation=<rot 0 -90 0>}]
 spine_05 [
  {shapeType=Sphere, radius=3.75, shapeName=spine_05_sphere, center=<12.4144 4.0822 -9.6298>}
  {shapeType=Sphere, radius=5.5792, shapeName=spine_05_sphere, center=<10.7947 2.3108 10.8295>}
  {
   shapeType=Capsule
   length=14.7028
   radius=8.2193
   shapeName=spine_05_capsule
   center=<6.4918 0.6751 0.0443>
   rotation=<rot 1.9269 -90 -1.9551>
  }
 ]
 upperarm_l [{shapeType=Capsule, length=21.84, radius=4.5441, shapeName=upperarm_l_capsule, center=<15.7888 0 0>, rotation=<rot 90 0 0>}]
 upperarm_r [{shapeType=Capsule, length=20.8378, radius=4.7523, shapeName=upperarm_r_capsule, center=<-14.2772 0 0>, rotation=<rot 90 0 0>}]
}
constraints [
 {
  bone1Name=head
  bone2Name=neck_02
  swing1LimitDegrees=45
  swing1Motion=Free
  swing2LimitDegrees=45
  swing2Motion=Free
  twistLimitDegrees=45
  twistMotion=Free
 }
 {
  bone1Name=neck_02
  bone2Name=neck_01
  swing1LimitDegrees=45
  swing1Motion=Free
  swing2LimitDegrees=45
  swing2Motion=Free
  twistLimitDegrees=45
  twistMotion=Free
 }
 {
  bone1Name=neck_01
  bone2Name=spine_05
  swing1LimitDegrees=45
  swing1Motion=Free
  swing2LimitDegrees=45
  swing2Motion=Free
  twistLimitDegrees=45
  twistMotion=Free
 }
 {bone1Name=dyn_hood, bone2Name=spine_05, swing1LimitDegrees=0, swing2LimitDegrees=0, twistLimitDegrees=0}
 {
  bone1Name=spine_05
  bone2Name=spine_04
  swing1LimitDegrees=45
  swing1Motion=Free
  swing2LimitDegrees=45
  swing2Motion=Free
  twistLimitDegrees=45
  twistMotion=Free
 }
 {
  bone1Name=spine_04
  bone2Name=spine_03
  swing1LimitDegrees=45
  swing1Motion=Free
  swing2LimitDegrees=45
  swing2Motion=Free
  twistLimitDegrees=45
  twistMotion=Free
 }
 {bone1Name=dyn_L_hair_back, bone2Name=head, swing1LimitDegrees=3, swing2LimitDegrees=2, twistLimitDegrees=0}
 {bone1Name=dyn_L_hair_side, bone2Name=head, swing1LimitDegrees=5, swing2LimitDegrees=5, twistLimitDegrees=0}
 {bone1Name=dyn_M_hair_front, bone2Name=head, swing1LimitDegrees=5, swing2LimitDegrees=5, twistLimitDegrees=0}
 {bone1Name=dyn_hair_top, bone2Name=head, swing1LimitDegrees=4, swing2LimitDegrees=4, twistLimitDegrees=0}
 {bone1Name=dyn_hood_top, bone2Name=hood_top_Master, swing1LimitDegrees=5, swing2LimitDegrees=5, twistLimitDegrees=2.5}
 {bone1Name=dyn_L_hair_back_down, bone2Name=dyn_L_hair_back, swing1LimitDegrees=10, swing2LimitDegrees=10, twistLimitDegrees=0}
 {bone1Name=dyn_L_hair_back_down_2, bone2Name=dyn_L_hair_back, swing1LimitDegrees=10, swing2LimitDegrees=10, twistLimitDegrees=0}
 {bone1Name=dyn_M_hair_back_Down, bone2Name=dyn_L_hair_back, swing1LimitDegrees=5, swing2LimitDegrees=5, twistLimitDegrees=0}
 {bone1Name=dyn_M_hair_back_mid, bone2Name=M_hair_back_mid_MASTER, swing1LimitDegrees=5, swing2LimitDegrees=5, twistLimitDegrees=0}
 {bone1Name=dyn_R_hair_back_down, bone2Name=dyn_L_hair_back, swing1LimitDegrees=10, swing2LimitDegrees=10, twistLimitDegrees=0}
 {bone1Name=dyn_R_hair_back_mid, bone2Name=R_hair_back_mid_MASTER, swing1LimitDegrees=5, swing2LimitDegrees=5, twistLimitDegrees=0}
 {
  bone1Name=upperarm_r
  bone2Name=clavicle_r
  swing1LimitDegrees=45
  swing1Motion=Free
  swing2LimitDegrees=45
  swing2Motion=Free
  twistLimitDegrees=45
  twistMotion=Free
 }
 {
  bone1Name=upperarm_l
  bone2Name=clavicle_l
  swing1LimitDegrees=45
  swing1Motion=Free
  swing2LimitDegrees=45
  swing2Motion=Free
  twistLimitDegrees=45
  twistMotion=Free
 }
 {bone1Name=dyn_hood_L_ear_1, bone2Name=dyn_hood, swing1LimitDegrees=20, swing2LimitDegrees=40, twistLimitDegrees=0}
 {bone1Name=dyn_hood_L_ear_2, bone2Name=dyn_hood_L_ear_1_1, swing1LimitDegrees=40, swing2LimitDegrees=40, twistLimitDegrees=0}
 {bone1Name=dyn_hood_R_ear_1, bone2Name=dyn_hood, swing1LimitDegrees=20, swing2LimitDegrees=20, twistLimitDegrees=0}
 {bone1Name=dyn_hood_R_ear_2, bone2Name=dyn_hood_R_ear_1_1, swing1LimitDegrees=40, swing2LimitDegrees=40, twistLimitDegrees=0}
 {
  bone1Name=M_hair_back_mid_MASTER
  bone2Name=head
  swing1LimitDegrees=45
  swing1Motion=Free
  swing2LimitDegrees=45
  swing2Motion=Free
  twistLimitDegrees=45
  twistMotion=Free
 }
 {
  bone1Name=R_hair_back_mid_MASTER
  bone2Name=head
  swing1LimitDegrees=45
  swing1Motion=Free
  swing2LimitDegrees=45
  swing2Motion=Free
  twistLimitDegrees=45
  twistMotion=Free
 }
 {
  bone1Name=hood_top_Master
  bone2Name=spine_05
  swing1LimitDegrees=45
  swing1Motion=Free
  swing2LimitDegrees=45
  swing2Motion=Free
  twistLimitDegrees=45
  twistMotion=Free
 }
 {
  bone1Name=hood_top2_Master
  bone2Name=spine_05
  swing1LimitDegrees=45
  swing1Motion=Free
  swing2LimitDegrees=45
  swing2Motion=Free
  twistLimitDegrees=45
  twistMotion=Free
 }
 {bone1Name=dyn_hood_top2, bone2Name=hood_top2_Master, swing1LimitDegrees=7, swing2LimitDegrees=7, twistLimitDegrees=3}
 {bone1Name=dyn_hood_R_ear_1_1, bone2Name=dyn_hood_R_ear_1, swing1LimitDegrees=40, swing2LimitDegrees=40, twistLimitDegrees=0}
 {bone1Name=dyn_hood_L_ear_1_1, bone2Name=dyn_hood_L_ear_1, swing1LimitDegrees=40, swing2LimitDegrees=40, twistLimitDegrees=0}
 {
  bone1Name=clavicle_l
  bone2Name=spine_05
  swing1LimitDegrees=45
  swing1Motion=Free
  swing2LimitDegrees=45
  swing2Motion=Free
  twistLimitDegrees=45
  twistMotion=Free
 }
 {
  bone1Name=clavicle_r
  bone2Name=spine_05
  swing1LimitDegrees=45
  swing1Motion=Free
  swing2LimitDegrees=45
  swing2Motion=Free
  twistLimitDegrees=45
  twistMotion=Free
 }
]
