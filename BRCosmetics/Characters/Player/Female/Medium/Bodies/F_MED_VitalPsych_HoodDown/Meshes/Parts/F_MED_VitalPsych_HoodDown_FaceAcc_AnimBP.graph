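# Post-processes the hood, hood ears, hair and ponytails of the VitalPsych HoodDown female character's face accessory mesh.
# Copied parent pose, with body-mesh arms layered in, passes through ponytail constraints, a Control Rig and hood constraints.
# Two rigid body simulations with separate physics assets then alternate on backbling presence, followed by additive corrective bone offsets.
# The event graph derives gravity and blend alphas from movement states, head and clavicle rotation, and backbling checks.
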AnimBlueprint /BRCosmetics/Characters/Player/Female/Medium/Bodies/F_MED_VitalPsych_HoodDown/Meshes/Parts/F_MED_VitalPsych_HoodDown_FaceAcc_AnimBP
target_skeleton=&/F_MED_VitalPsych_HoodDown_FaceAcc_Skeleton
graphs [
 AnimGraph {nodes [
   n1 = Get GravityOverride()
   n2 = RigidBody!(
    ComponentPose=Local To Component(LocalPose=Use cached pose 'Constraints Cache'())
    Alpha="Get Wearing Backbling "() -> ("Wearing Backbling ":float)
    OverrideWorldGravity:Vector=@n1
    ExternalForce
   ) -> (Pose) {
    bOverrideWorldGravity=true
    overridePhysicsAsset=&/F_MED_VitalPsych_HoodDown_FaceAcc_Physics2
    simulationSpace=BaseBoneSpace
    alphaBoolBlend {blendOption=Linear}
    baseBoneRef {boneName=spine_05}
    componentAppliedLinearAccClamp=<3000 3000 3000>
    componentLinearAccScale=<0.2 0.25 0.3>
    componentLinearVelScale=<0.2 0.25 0.3>
    overrideWorldGravity {z=-980}
    simSpaceSettings {
     maxAngularAcceleration=1000
     maxAngularVelocity=3
     maxLinearAcceleration=700
     maxLinearVelocity=500
     velocityScaleZ=0
     worldAlpha=0.4
     externalLinearDragV=<0.2 0.2>
    }
   }
   n3 = RigidBody(ComponentPose=@n2, Alpha=Get Not Wearing Backbling() -> (Not Wearing Backbling:float), OverrideWorldGravity=@n1) {
    bOverrideWorldGravity=true
    overridePhysicsAsset=&/F_MED_VitalPsych_HoodDown_FaceAcc_NO_BP_Physics
    simulationSpace=BaseBoneSpace
    alphaBoolBlend {blendOption=Linear}
    baseBoneRef {boneName=spine_05}
    componentAppliedLinearAccClamp=<3000 3000 3000>
    componentLinearAccScale=<0.2 0.25 0.3>
    componentLinearVelScale=<0.2 0.25 0.3>
    overrideWorldGravity {z=-980}
    simSpaceSettings {
     maxAngularAcceleration=1000
     maxAngularVelocity=3
     maxLinearAcceleration=700
     maxLinearVelocity=500
     velocityScaleZ=0
     worldAlpha=0.7
     externalLinearDragV=<0.2 0.2>
    }
   }
   n4 = Get Clavicle_L_rot() -> (Clavicle_L_rot:double)
   n5 = "Transform (Modify) Bone - Bone: dyn_hood_L_ear_1"(ComponentPose=@n3, Alpha=@n4, Translation="0,0,0", Rotation="-10,0,0", Scale="1,1,1") {
    rotationMode=BMM_Additive
    rotationSpace=BCS_BoneSpace
    alphaBoolBlend {blendOption=Linear}
    boneToModify {boneName=dyn_hood_L_ear_1}
    rotation {pitch=-10}
   }
   n6 = "Transform (Modify) Bone - Bone: dyn_hood"(ComponentPose=@n5, Alpha=@n4, Translation="0,0,0", Rotation="-5,0,0", Scale="1,1,1") {
    rotationMode=BMM_Additive
    rotationSpace=BCS_BoneSpace
    alphaBoolBlend {blendOption=Linear}
    boneToModify {boneName=dyn_hood}
    rotation {pitch=-5}
   }
   n7 = "Transform (Modify) Bone - Bone: dyn_hood_R_ear_1"(ComponentPose=@n6, Alpha=Get Clavicle_R_rot() -> (Clavicle_R_rot:double), Translation="0,0,0", Rotation="-20,0,0", Scale="1,1,1") {
    rotationMode=BMM_Additive
    rotationSpace=BCS_BoneSpace
    alphaBoolBlend {blendOption=Linear}
    boneToModify {boneName=dyn_hood_R_ear_1}
    rotation {pitch=-20}
   }
   n8 = Get is Falling Jumping Sprinting() -> (is Falling Jumping Sprinting:float)
   n9 = "Transform (Modify) Bone - Bone: dyn_hood"(ComponentPose=@n7, Alpha=@n8, Translation="2,-0.314,-1.5", Rotation="-2,0,0", Scale="1,1,1") {
    rotationMode=BMM_Additive
    rotationSpace=BCS_BoneSpace
    translationMode=BMM_Additive
    translationSpace=BCS_BoneSpace
    alphaBoolBlend {blendOption=Linear}
    alphaScaleBiasClamp {bInterpResult=true, interpSpeedDecreasing=3, interpSpeedIncreasing=3}
    boneToModify {boneName=dyn_hood}
    rotation {pitch=-2}
    translation=<2 -0.314 -1.5>
   }
   n10 = "Transform (Modify) Bone - Bone: dyn_hood_L_ear_1"(ComponentPose=@n9, Alpha=@n8, Translation="0,0,-2", Rotation="-5,0,0", Scale="1,1,1") {
    rotationMode=BMM_Additive
    rotationSpace=BCS_BoneSpace
    translationMode=BMM_Additive
    translationSpace=BCS_BoneSpace
    alphaBoolBlend {blendOption=Linear}
    alphaScaleBiasClamp {bInterpResult=true}
    boneToModify {boneName=dyn_hood_L_ear_1}
    rotation {pitch=-5}
    translation {z=-2}
   }
   n11 = "Transform (Modify) Bone - Bone: dyn_hood_R_ear_1"(ComponentPose=@n10, Alpha=@n8, Translation="0,0,2", Rotation="5,0,0", Scale="1,1,1") {
    rotationMode=BMM_Additive
    rotationSpace=BCS_BoneSpace
    translationMode=BMM_Additive
    translationSpace=BCS_BoneSpace
    alphaBoolBlend {blendOption=Linear}
    alphaScaleBiasClamp {bInterpResult=true}
    boneToModify {boneName=dyn_hood_R_ear_1}
    rotation {pitch=5}
    translation {z=2}
   }
   n12 = "Transform (Modify) Bone - Bone: dyn_L_hair_back"(ComponentPose=@n11, Alpha=Get Skydive Floating() -> (Skydive Floating:float), Translation="0,0,0", Rotation="20,10,0", Scale="1,1,1") {
    rotationMode=BMM_Additive
    rotationSpace=BCS_BoneSpace
    scaleSpace=BCS_BoneSpace
    translationSpace=BCS_BoneSpace
    alphaBoolBlend {blendOption=Linear}
    alphaScaleBiasClamp {bInterpResult=true, interpSpeedDecreasing=5, interpSpeedIncreasing=5}
    boneToModify {boneName=dyn_L_hair_back}
    rotation {pitch=20, yaw=10}
   }
   n13 = Get Parachute opened() -> (Parachute opened:float)
   n14 = "Transform (Modify) Bone - Bone: Hood_shoulder_fix_L"(ComponentPose=@n12, Alpha=@n13, Translation="-1,0,0", Rotation="20,0,0", Scale="1,1,1") {
    rotationMode=BMM_Additive
    rotationSpace=BCS_BoneSpace
    scaleSpace=BCS_BoneSpace
    translationMode=BMM_Additive
    translationSpace=BCS_BoneSpace
    alphaBoolBlend {blendOption=Linear}
    alphaScaleBiasClamp {bInterpResult=true}
    boneToModify {boneName=Hood_shoulder_fix_L}
    rotation {pitch=20}
    translation {x=-1}
   }
   n15 = "Transform (Modify) Bone - Bone: Hood_shoulder_fix_R"(ComponentPose=@n14, Alpha=@n13, Translation="1,0,0", Rotation="20,0,0", Scale="1,1,1") {
    rotationMode=BMM_Additive
    rotationSpace=BCS_BoneSpace
    scaleSpace=BCS_BoneSpace
    translationMode=BMM_Additive
    translationSpace=BCS_BoneSpace
    alphaBoolBlend {blendOption=Linear}
    alphaScaleBiasClamp {bInterpResult=true}
    boneToModify {boneName=Hood_shoulder_fix_R}
    rotation {pitch=20}
    translation {x=1}
   }
   n16 = Get Falling without BP() -> (Falling without BP:float)
   n17 = "Transform (Modify) Bone - Bone: dyn_hood_L_ear_1_1"(ComponentPose=@n15, Alpha=@n16, Translation="0,0,0", Rotation="-50,0,0", Scale="1,1,1") {
    rotationMode=BMM_Additive
    rotationSpace=BCS_BoneSpace
    alphaBoolBlend {blendOption=Linear}
    alphaScaleBiasClamp {bInterpResult=true, interpSpeedIncreasing=8}
    boneToModify {boneName=dyn_hood_L_ear_1_1}
    rotation {pitch=-50}
   }
   n18 = "Transform (Modify) Bone - Bone: dyn_hood_L_ear_2"(ComponentPose=@n17, Alpha=@n16, Translation="0,0,0", Rotation="-40,0,0", Scale="1,1,1") {
    rotationMode=BMM_Additive
    rotationSpace=BCS_BoneSpace
    alphaBoolBlend {blendOption=Linear}
    alphaScaleBiasClamp {bInterpResult=true, interpSpeedDecreasing=7, interpSpeedIncreasing=8}
    boneToModify {boneName=dyn_hood_L_ear_2}
    rotation {pitch=-40}
   }
   n19 = "Transform (Modify) Bone - Bone: dyn_hood_R_ear_1_1"(ComponentPose=@n18, Alpha=@n16, Translation="0,0,0", Rotation="50,0,0", Scale="1,1,1") {
    rotationMode=BMM_Additive
    rotationSpace=BCS_BoneSpace
    alphaBoolBlend {blendOption=Linear}
    alphaScaleBiasClamp {bInterpResult=true, interpSpeedIncreasing=8}
    boneToModify {boneName=dyn_hood_R_ear_1_1}
    rotation {pitch=50}
   }
   n20 = "Transform (Modify) Bone - Bone: dyn_hood_R_ear_2"(ComponentPose=@n19, Alpha=@n16, Translation="0,0,0", Rotation="40,0,0", Scale="1,1,1") {
    rotationMode=BMM_Additive
    rotationSpace=BCS_BoneSpace
    alphaBoolBlend {blendOption=Linear}
    alphaScaleBiasClamp {bInterpResult=true, interpSpeedDecreasing=7, interpSpeedIncreasing=8}
    boneToModify {boneName=dyn_hood_R_ear_2}
    rotation {pitch=40}
   }
   n21 = Get Skydive floating without BP() -> (Skydive floating without BP:float)
   n22 = "Transform (Modify) Bone - Bone: dyn_hood_L_ear_1_1"(ComponentPose=@n20, Alpha=@n21, Translation="0,0,0", Rotation="-40,0,0", Scale="1,1,1") {
    rotationMode=BMM_Additive
    rotationSpace=BCS_BoneSpace
    alphaBoolBlend {blendOption=Linear}
    alphaScaleBiasClamp {bInterpResult=true, interpSpeedIncreasing=8}
    boneToModify {boneName=dyn_hood_L_ear_1_1}
    rotation {pitch=-40}
   }
   n23 = "Transform (Modify) Bone - Bone: dyn_hood_R_ear_1_1"(ComponentPose=@n22, Alpha=@n21, Translation="0,0,0", Rotation="40,0,0", Scale="1,1,1") {
    rotationMode=BMM_Additive
    rotationSpace=BCS_BoneSpace
    alphaBoolBlend {blendOption=Linear}
    alphaScaleBiasClamp {bInterpResult=true, interpSpeedIncreasing=8}
    boneToModify {boneName=dyn_hood_R_ear_1_1}
    rotation {pitch=40}
   }
   n24 = "Transform (Modify) Bone - Bone: dyn_hood_L_ear_1"(ComponentPose=@n23, Alpha=@n21, Translation="0,0,0", Rotation="-20,0,0", Scale="1,1,1") {
    rotationMode=BMM_Additive
    rotationSpace=BCS_BoneSpace
    alphaBoolBlend {blendOption=Linear}
    alphaScaleBiasClamp {bInterpResult=true, interpSpeedIncreasing=8}
    boneToModify {boneName=dyn_hood_L_ear_1}
    rotation {pitch=-20}
   }
   n25 = "Transform (Modify) Bone - Bone: dyn_hood_R_ear_1"(ComponentPose=@n24, Alpha=@n21, Translation="0,0,0", Rotation="20,0,0", Scale="1,1,1") {
    rotationMode=BMM_Additive
    rotationSpace=BCS_BoneSpace
    alphaBoolBlend {blendOption=Linear}
    alphaScaleBiasClamp {bInterpResult=true, interpSpeedIncreasing=8}
    boneToModify {boneName=dyn_hood_R_ear_1}
    rotation {pitch=20}
   }
   n26 = Get Look Up Alpha HAIR() -> (Look Up Alpha HAIR:double)
   n27 = "Transform (Modify) Bone - Bone: dyn_L_hair_back_down_2"(ComponentPose=@n25, Alpha=@n26, Translation="0,0,0", Rotation="12,15.421,-0.814", Scale="1,1,1") {
    rotationMode=BMM_Additive
    rotationSpace=BCS_BoneSpace
    alphaBoolBlend {blendOption=Linear}
    alphaScaleBiasClamp {bInterpResult=true, interpSpeedDecreasing=5, interpSpeedIncreasing=3}
    boneToModify {boneName=dyn_L_hair_back_down_2}
    rotation=<rot 12 15.421 -0.814>
   }
   n28 = "Transform (Modify) Bone - Bone: dyn_L_hair_back"(ComponentPose=@n27, Alpha=@n26, Translation="0,0.8,0", Rotation="0,0,0", Scale="1,1,1") {
    rotationMode=BMM_Additive
    rotationSpace=BCS_BoneSpace
    translationMode=BMM_Additive
    alphaBoolBlend {blendOption=Linear}
    alphaScaleBiasClamp {bInterpResult=true, interpSpeedDecreasing=5}
    boneToModify {boneName=dyn_L_hair_back}
    translation {y=0.8}
   }
   n29 = "Transform (Modify) Bone - Bone: dyn_R_hair_back_down"(ComponentPose=@n28, Alpha=Get Hood jump Falling() -> (Hood jump Falling:float), Translation="0,0,0", Rotation="20,-10,0", Scale="1,1,1") {
    rotationMode=BMM_Additive
    rotationSpace=BCS_BoneSpace
    alphaBoolBlend {blendOption=Linear}
    alphaScaleBiasClamp {bInterpResult=true, interpSpeedDecreasing=5}
    boneToModify {boneName=dyn_R_hair_back_down}
    rotation {pitch=20, yaw=-10}
   }
   n30 = "Get Wearing Backbling "()
   n31 = "Transform (Modify) Bone - Bone: dyn_hood_L_ear_1_1"(ComponentPose=@n29, Alpha=@n30, Translation="0,0,0", Rotation="5,0,0", Scale="1,1,1") {
    rotationMode=BMM_Additive
    rotationSpace=BCS_BoneSpace
    alphaBoolBlend {blendOption=Linear}
    alphaScaleBiasClamp {bInterpResult=true, interpSpeedDecreasing=5}
    boneToModify {boneName=dyn_hood_L_ear_1_1}
    rotation {pitch=5}
   }
   n32 = "Transform (Modify) Bone - Bone: dyn_hood_R_ear_1_1"(ComponentPose=@n31, Alpha=@n30, Translation="0,0,0", Rotation="-5,0,0", Scale="1,1,1") {
    rotationMode=BMM_Additive
    rotationSpace=BCS_BoneSpace
    alphaBoolBlend {blendOption=Linear}
    alphaScaleBiasClamp {bInterpResult=true, interpSpeedDecreasing=5}
    boneToModify {boneName=dyn_hood_R_ear_1_1}
    rotation {pitch=-5}
   }
   n33 = Get falling playing melee() -> (falling playing melee:float)
   n34 = "Transform (Modify) Bone - Bone: dyn_hood_R_ear_1"(ComponentPose=@n32, Alpha=@n33, Translation="0,0,0", Rotation="-10,0,0", Scale="1,1,1") {
    rotationMode=BMM_Additive
    rotationSpace=BCS_BoneSpace
    alphaBoolBlend {blendOption=Linear}
    alphaScaleBiasClamp {bInterpResult=true, interpSpeedDecreasing=5}
    boneToModify {boneName=dyn_hood_R_ear_1}
    rotation {pitch=-10}
   }
   n35 = "Transform (Modify) Bone - Bone: dyn_hood_L_ear_1"(ComponentPose=@n34, Alpha=@n33, Translation="0,0,0", Rotation="10,0,0", Scale="1,1,1") {
    rotationMode=BMM_Additive
    rotationSpace=BCS_BoneSpace
    alphaBoolBlend {blendOption=Linear}
    alphaScaleBiasClamp {bInterpResult=true, interpSpeedDecreasing=5}
    boneToModify {boneName=dyn_hood_L_ear_1}
    rotation {pitch=10}
   }
   Output Pose(Result=Component To Local(ComponentPose=@n35)) {layerGroup=None}
   n36 = "Constraint - base_ponytail_00_l (PPP)"!(
    ComponentPose=Local To Component(LocalPose=Use cached pose 'InputAnimation'())
    Alpha=1
    ConstraintWeights_0=0.6623
    ConstraintWeights_1:float=0.2316
    ConstraintWeights_2:float=0.1061
   ) -> (Pose) {
    alphaBoolBlend {blendOption=Linear}
    boneToModify {boneName=base_ponytail_00_l}
    constraintSetup [{targetBone {boneName=head}}, {targetBone {boneName=neck_02}}, {targetBone {boneName=neck_01}}]
    constraintWeights [1,1,1]
   }
   n37 = "Constraint - dyn_ponytail_01_l (PPP)"!(ComponentPose=@n36, Alpha=1, ConstraintWeights_0=0.6341, ConstraintWeights_1:float=0.3371, ConstraintWeights_2:float=0.0288) -> (Pose) {
    alphaBoolBlend {blendOption=Linear}
    boneToModify {boneName=dyn_ponytail_01_l}
    constraintSetup [{targetBone {boneName=head}}, {targetBone {boneName=neck_02}}, {targetBone {boneName=neck_01}}]
    constraintWeights [1,1,1]
   }
   n38 = "Constraint - dyn_ponytail_01_l (PPPPP)"!(
    ComponentPose=@n37
    Alpha=0.5
    ConstraintWeights_0=0.7156
    ConstraintWeights_1:float=0.2634
    ConstraintWeights_2:float=0.0162
    ConstraintWeights_3:float=0.0032
    ConstraintWeights_4:float=0.0016
   ) -> (Pose) {
    alphaBoolBlend {blendOption=Linear}
    boneToModify {boneName=dyn_ponytail_01_l}
    constraintSetup [
     {targetBone {boneName=spine_05}}
     {targetBone {boneName=clavicle_l}}
     {targetBone {boneName=spine_04}}
     {targetBone {boneName=upperarm_twist_01_l}}
     {targetBone {boneName=neck_02}}
    ]
    constraintWeights [1,1,1,1,1]
   }
   n39 = "Constraint - dyn_ponytail_02_l (PPP)"!(ComponentPose=@n38, Alpha=0.5, ConstraintWeights_0=0.7888, ConstraintWeights_1:float=0.1818, ConstraintWeights_2:float=0.0294) -> (Pose) {
    alphaBoolBlend {blendOption=Linear}
    boneToModify {boneName=dyn_ponytail_02_l}
    constraintSetup [{targetBone {boneName=spine_05}}, {targetBone {boneName=clavicle_l}}, {targetBone {boneName=spine_04}}]
    constraintWeights [1,1,1]
   }
   n40 = "Constraint - base_ponytail_00_r (PPP)"!(ComponentPose=@n39, Alpha=1, ConstraintWeights_0=0.6623, ConstraintWeights_1:float=0.2316, ConstraintWeights_2:float=0.1061) -> (Pose) {
    alphaBoolBlend {blendOption=Linear}
    boneToModify {boneName=base_ponytail_00_r}
    constraintSetup [{targetBone {boneName=head}}, {targetBone {boneName=neck_02}}, {targetBone {boneName=neck_01}}]
    constraintWeights [1,1,1]
   }
   n41 = "Constraint - dyn_ponytail_01_r (PPP)"!(ComponentPose=@n40, Alpha=1, ConstraintWeights_0=0.6341, ConstraintWeights_1:float=0.3371, ConstraintWeights_2:float=0.0288) -> (Pose) {
    alphaBoolBlend {blendOption=Linear}
    boneToModify {boneName=dyn_ponytail_01_r}
    constraintSetup [{targetBone {boneName=head}}, {targetBone {boneName=neck_02}}, {targetBone {boneName=neck_01}}]
    constraintWeights [1,1,1]
   }
   n42 = "Constraint - dyn_ponytail_01_r (PPPPP)"!(
    ComponentPose=@n41
    Alpha=0.35
    ConstraintWeights_0=0.7156
    ConstraintWeights_1:float=0.2634
    ConstraintWeights_2:float=0.0162
    ConstraintWeights_3:float=0.0032
    ConstraintWeights_4:float=0.0016
   ) -> (Pose) {
    alphaBoolBlend {blendOption=Linear}
    boneToModify {boneName=dyn_ponytail_01_r}
    constraintSetup [
     {targetBone {boneName=spine_05}}
     {targetBone {boneName=clavicle_r}}
     {targetBone {boneName=spine_04}}
     {targetBone {boneName=upperarm_twist_01_r}}
     {targetBone {boneName=neck_02}}
    ]
    constraintWeights [1,1,1,1,1]
   }
   n43 = "Constraint - dyn_ponytail_02_r (PPPP)"!(
    ComponentPose=@n42
    Alpha=0.5
    ConstraintWeights_0=0.8527
    ConstraintWeights_1:float=0.1265
    ConstraintWeights_2:float=0.0208
    ConstraintWeights_3:float=0.3
   ) -> (Pose) {
    alphaBoolBlend {blendOption=Linear}
    boneToModify {boneName=dyn_ponytail_02_r}
    constraintSetup [{targetBone {boneName=spine_05}}, {targetBone {boneName=clavicle_r}}, {targetBone {boneName=spine_04}}, {targetBone {boneName=head}}]
    constraintWeights [1,1,1,1]
   }
   n44 = "Transform (Modify) Bone - Bone: dyn_ponytail_00_l"(ComponentPose=@n43, Alpha=1, Translation="0,0,0", Rotation="-5.000,-15.000,0.000", Scale="1,1,1") {
    rotationMode=BMM_Additive
    rotationSpace=BCS_BoneSpace
    alphaBoolBlend {blendOption=Linear}
    boneToModify {boneName=dyn_ponytail_00_l}
    rotation=<rot -5 -14.9999 0>
   }
   n45 = "Transform (Modify) Bone - Bone: dyn_ponytail_00_r"(ComponentPose=@n44, Alpha=1, Translation="0,0,0", Rotation="10.000,-15.000,0.000", Scale="1,1,1") {
    rotationMode=BMM_Additive
    rotationSpace=BCS_BoneSpace
    alphaBoolBlend {blendOption=Linear}
    boneToModify {boneName=dyn_ponytail_00_r}
    rotation=<rot 10 -15 0>
   }
   n46 = "Transform (Modify) Bone - Bone: dyn_ponytail_01_l"(ComponentPose=@n45, Alpha=1, Translation="0,0,0", Rotation="0,-25,0", Scale="1,1,1") {
    rotationMode=BMM_Additive
    rotationSpace=BCS_BoneSpace
    alphaBoolBlend {blendOption=Linear}
    boneToModify {boneName=dyn_ponytail_01_l}
    rotation {yaw=-25}
   }
   n47 = "Transform (Modify) Bone - Bone: dyn_ponytail_01_r"(ComponentPose=@n46, Alpha=1, Translation="0,0,0", Rotation="0,-25,0", Scale="1,1,1") {
    rotationMode=BMM_Additive
    rotationSpace=BCS_BoneSpace
    alphaBoolBlend {blendOption=Linear}
    boneToModify {boneName=dyn_ponytail_01_r}
    rotation {yaw=-25}
   }
   n48 = "Transform (Modify) Bone - Bone: dyn_ponytail_04_r"!(ComponentPose=@n47, Alpha=1, Scale="0,0,0") -> (Pose) {
    rotationSpace=BCS_BoneSpace
    scaleMode=BMM_Additive
    alphaBoolBlend {blendOption=Linear}
    boneToModify {boneName=dyn_ponytail_04_r}
    scale=<0 0 0>
   }
   n49 = "Transform (Modify) Bone - Bone: dyn_ponytail_04_l"!(ComponentPose=@n48, Alpha=1, Scale="0,0,0") -> (Pose) {
    rotationSpace=BCS_BoneSpace
    scaleMode=BMM_Additive
    alphaBoolBlend {blendOption=Linear}
    boneToModify {boneName=dyn_ponytail_04_l}
    scale=<0 0 0>
   }
   Save cached pose 'LongHair_Dyn_Cache'(Pose=Component To Local(ComponentPose=@n49)) {cacheName=LongHair_Dyn_Cache}
   n50 = Get Constraints Alpha() -> (Constraints Alpha:float)
   n51 = Control Rig(Source=Use cached pose 'LongHair_Dyn_Cache'()) {
    alphaBoolBlend {blendOption=Linear}
    controlRigAssetReference {blueprintRigClass=&/F_MED_VitalPsych_HoodDown_FaceAcc_CtrlRig.F_MED_VitalPsych_HoodDown_FaceAcc_CtrlRig_C}
   }
   n52 = "Constraint - hood_top_Master (RR)"!(
    ComponentPose=Local To Component(LocalPose=@n51)
    Alpha=@n50
    ConstraintWeights_0=Get head : Rotation() -> ("head : Rotation":float)
    ConstraintWeights_1:float=Get spine_05 : Rotation() -> ("spine_05 : Rotation":float)
   ) -> (Pose) {
    alphaBoolBlend {blendOption=Linear}
    alphaScaleBiasClamp {bInterpResult=true, interpSpeedDecreasing=7}
    boneToModify {boneName=hood_top_Master}
    constraintSetup [{transformType=Rotation, targetBone {boneName=head}}, {transformType=Rotation, targetBone {boneName=spine_05}}]
    constraintWeights [1,1]
   }
   n53 = "Constraint - hood_top2_Master (PP)"!(ComponentPose=@n52, Alpha=@n50, ConstraintWeights_0=0.7, ConstraintWeights_1:float=0.3) -> (Pose) {
    alphaBoolBlend {blendOption=Linear}
    alphaScaleBiasClamp {bInterpResult=true, interpSpeedDecreasing=5}
    boneToModify {boneName=hood_top2_Master}
    constraintSetup [{targetBone {boneName=hood_top_Master}}, {targetBone {boneName=spine_05}}]
    constraintWeights [1,1]
   }
   n54 = "Transform (Modify) Bone - Bone: dyn_hood_inside"(ComponentPose=@n53, Alpha=Get Look Up Alpha HAIR(), Translation="0,0,2", Rotation="0,0,0", Scale="1,1,1") {
    rotationSpace=BCS_BoneSpace
    translationMode=BMM_Additive
    translationSpace=BCS_BoneSpace
    alphaBoolBlend {blendOption=Linear}
    alphaScaleBiasClamp {bInterpResult=true, interpSpeedDecreasing=5}
    boneToModify {boneName=dyn_hood_inside}
    translation {z=2}
   }
   n55 = "Constraint - dyn_hood (RR)"!(
    ComponentPose=@n54
    Alpha=Get Not playing Malee Anim() -> (Not playing Malee Anim:float)
    ConstraintWeights_0=0.1
    ConstraintWeights_1:float=0.9
   ) -> (Pose) {
    alphaBoolBlend {blendOption=Linear}
    alphaScaleBiasClamp {bInterpResult=true}
    boneToModify {boneName=dyn_hood}
    constraintSetup [{transformType=Rotation, targetBone {boneName=head}}, {transformType=Rotation, targetBone {boneName=spine_05}}]
    constraintWeights [1,1]
   }
   n56 = "Constraint - Hood_shoulder_fix_L (PP)"!(ComponentPose=@n55, Alpha=1, ConstraintWeights_0=0.9, ConstraintWeights_1:float=0.1) -> (Pose) {
    alphaBoolBlend {blendOption=Linear}
    alphaScaleBiasClamp {bInterpResult=true}
    boneToModify {boneName=Hood_shoulder_fix_L}
    constraintSetup [{perAxis {bX=false, bY=false}, targetBone {boneName=clavicle_l}}, {perAxis {bX=false, bY=false}, targetBone {boneName=spine_05}}]
    constraintWeights [1,1]
   }
   n57 = "Constraint - Hood_shoulder_fix_R (RR)"!(ComponentPose=@n56, Alpha=1, ConstraintWeights_0=0.9, ConstraintWeights_1:float=0.1) -> (Pose) {
    alphaBoolBlend {blendOption=Linear}
    alphaScaleBiasClamp {bInterpResult=true}
    boneToModify {boneName=Hood_shoulder_fix_R}
    constraintSetup [{transformType=Rotation, targetBone {boneName=clavicle_r}}, {transformType=Rotation, targetBone {boneName=spine_05}}]
    constraintWeights [1,1]
   }
   n58 = "Constraint - dyn_M_hair_back_Down (PP)"!(ComponentPose=@n57, Alpha=1, ConstraintWeights_0=0, ConstraintWeights_1:float=1) -> (Pose) {
    alphaBoolBlend {blendOption=Linear}
    alphaScaleBiasClamp {bInterpResult=true}
    boneToModify {boneName=dyn_M_hair_back_Down}
    constraintSetup [{targetBone {boneName=head}}, {targetBone {boneName=spine_05}}]
    constraintWeights [1,1]
   }
   n59 = "Transform (Modify) Bone - Bone: hood_top_Master"(ComponentPose=@n58, Alpha=1, Translation="2,0,2", Rotation="-6.624,0,0", Scale="1,1,1") {
    rotationMode=BMM_Additive
    translationMode=BMM_Additive
    translationSpace=BCS_BoneSpace
    alphaBoolBlend {blendOption=Linear}
    boneToModify {boneName=hood_top_Master}
    rotation {pitch=-6.624}
    translation {x=2, z=2}
   }
   n60 = Get Look Up Alpha HAIR()
   n61 = "Transform (Modify) Bone - Bone: M_hair_back_mid_MASTER"(ComponentPose=@n59, Alpha=@n60, Translation="0,0,0", Rotation="22.554,-0.000,-0.000", Scale="1,1,1") {
    rotationMode=BMM_Additive
    rotationSpace=BCS_BoneSpace
    alphaBoolBlend {blendOption=Linear}
    alphaScaleBiasClamp {bInterpResult=true, interpSpeedDecreasing=5}
    boneToModify {boneName=M_hair_back_mid_MASTER}
    rotation=<rot 22.554 0 0>
   }
   n62 = "Transform (Modify) Bone - Bone: R_hair_back_mid_MASTER"(ComponentPose=@n61, Alpha=@n60, Translation="0,0.7,0", Rotation="40,0,0", Scale="1,1,1") {
    rotationMode=BMM_Additive
    rotationSpace=BCS_BoneSpace
    translationMode=BMM_Additive
    alphaBoolBlend {blendOption=Linear}
    alphaScaleBiasClamp {bInterpResult=true, interpSpeedDecreasing=5}
    boneToModify {boneName=R_hair_back_mid_MASTER}
    rotation {pitch=40}
    translation {y=0.7}
   }
   n63 = "Transform (Modify) Bone - Bone: hood_top_Master"(
    ComponentPose=@n62
    Alpha=Get hood jump falling No BP() -> (hood jump falling No BP:float)
    Translation="4,4,2"
    Rotation="0,0,0"
    Scale="1,1,1"
   ) {
    rotationMode=BMM_Additive
    translationMode=BMM_Additive
    translationSpace=BCS_BoneSpace
    alphaBoolBlend {blendOption=Linear}
    alphaScaleBiasClamp {bInterpResult=true}
    boneToModify {boneName=hood_top_Master}
    translation=<4 4 2>
   }
   n64 = "Transform (Modify) Bone - Bone: hood_top_Master"(ComponentPose=@n63, Alpha=Get Hood Sprint() -> (Hood Sprint:float), Translation="5,0,0", Rotation="0,0,-3", Scale="1,1,1") {
    rotationMode=BMM_Additive
    translationMode=BMM_Additive
    translationSpace=BCS_BoneSpace
    alphaBoolBlend {blendOption=Linear}
    alphaScaleBiasClamp {bInterpResult=true, interpSpeedDecreasing=5}
    boneToModify {boneName=hood_top_Master}
    rotation {roll=-3}
    translation {x=5}
   }
   Save cached pose 'Constraints Cache'(Pose=Component To Local(ComponentPose=@n64)) {cacheName=Constraints Cache}
   n65 = Layered blend per bone(
    BasePose=Copy Pose From Mesh() {bCopyCurves=true, bUseAttachedParent=true}
    BlendPoses_0=Copy Pose From Mesh(SourceMeshComponent=Get Body_Mesh())
   ) {layerSetup [{branchFilters [{boneName=upperarm_l}, {boneName=upperarm_r}, {boneName=clavicle_r}, {boneName=clavicle_l}]}]}
   Save cached pose 'InputAnimation'(Pose=@n65) {cacheName=InputAnimation}
  ]}
 EventGraph {nodes [
   Event BlueprintUpdateAnimation()
   > n1 = Sequence!(execute) -> (then_0, then_1, then_2:Exec, then_3:Exec, then_4:Exec, then_5:Exec, then_6:Exec)
   Set Hood Top Translation(execute=@n1.then_1, Hood Top Translation:Vector=Get Hood Top BP Translation() -> (Hood Top BP Translation:Vector)) -> (Output_Get:Vector)
   > Set HoodTop2 Translation(HoodTop2 Translation:Vector=Get Hood Top 2 BP Translation() -> (Hood Top 2 BP Translation:Vector)) -> (Output_Get:Vector)
   > Set HoodTop Rotation(HoodTop Rotation:Rotator=Get Hood Top BP Rotation() -> (Hood Top BP Rotation:Rotator)) -> (Output_Get:Rotator)
   > Set HoodTop2 Rotation(HoodTop2 Rotation:Rotator=Get Hood Top 2 BP Rotation() -> (Hood Top 2 BP Rotation:Rotator)) -> (Output_Get:Rotator)
   n2 = GetOwnerSkeletalMeshComponent()
   n3 = BreakRotator(InRot=BreakTransform(InTransform=GetDeltaTransformFromRefPose(self=@n2, BoneName=head, BaseName=spine_05)).Rotation)
   n4 = Branch(execute=@n1.then_3, Condition="float <= float"(A=@n3.Yaw, B=-30) -> (ReturnValue=~))
   n5 = Get Look Up Alpha_true() -> (Look Up Alpha_true:double)
   > Set Look up Alpha(Look up Alpha:double=@n5) -> (Output_Get:double)
   n6 = Branch(execute=@n4.else, Condition=AND Boolean(A="float <= float"(A=@n3.Yaw, B=-5) -> (ReturnValue=~), B=Get bIsTargeting()))
   > Set Look up Alpha(Look up Alpha=@n5)
   n7 = Get Look Up Alpha_false() -> (Look Up Alpha_false:double)
   Set Look up Alpha(execute=@n6.else, Look up Alpha=@n7)
   n8 = Branch(execute=@n1.then_5, Condition="float <= float"(A=@n3.Yaw, B=-10) -> (ReturnValue=~))
   > Set Look Up Alpha HAIR(Look Up Alpha HAIR:double=@n5) -> (Output_Get:double)
   Set Look Up Alpha HAIR(execute=@n8.else, Look Up Alpha HAIR=@n7)
   n9 = Gravity Override State Function DEPRECATED(
    execute=@n1.then_0
    CharacterPartAnimInstance=Self-Reference()
    GameTimeInSeconds=GetGameTimeInSeconds()
    GravityOverrideParams=Get Gravity Override Params()
   )
   > Set GravityOverride(GravityOverride=@n9)
   > Set Clavicle_L_rot(Clavicle_L_rot:double=GetAlphaFromReferenceTransformDelta(
     SkeletalMeshComponent=Get Skel Mesh()
     BoneName=spine_05
     BaseName=clavicle_l
     DeltaComponentType=EulerX
     InputRange=(X=0,Y=-35)
     TwistAxis="0,0,1"
    )) -> (Output_Get:double)
   > Set Clavicle_R_rot(Clavicle_R_rot:double=GetAlphaFromReferenceTransformDelta(
     SkeletalMeshComponent=Get Skel Mesh()
     BoneName=spine_05
     BaseName=clavicle_r
     DeltaComponentType=EulerX
     InputRange=(X=0,Y=-35)
     TwistAxis="0,0,1"
    )) -> (Output_Get:double)
   n10 = Branch(
    execute=@n1.then_2
    Condition=OR Boolean!(A=Get bIsCrouching(), B=Get bIsSwimSprinting() -> (bIsSwimSprinting:bool), C:bool=Get bIsSkydiveFloating(), D:bool=Get bIsSkyDiving()) -> (ReturnValue)
   )
   > Set head : Rotation("head : Rotation":float=Get Head : Rotation Crouching() -> ("Head : Rotation Crouching":float)) -> (Output_Get:float)
   > Set spine_05 : Rotation("spine_05 : Rotation":float=Get Spine 05 : Rotation Crouching() -> ("Spine 05 : Rotation Crouching":float)) -> (Output_Get:float)
   Set head : Rotation(execute=@n10.else, "head : Rotation"=Get Head : Rotation_skydive() -> ("Head : Rotation_skydive":float))
   > Set spine_05 : Rotation("spine_05 : Rotation"=Get Spine 05 : Rotation Skydive() -> ("Spine 05 : Rotation Skydive":float))
   n11 = Sequence!(execute=@n1.then_4) -> (then_0, then_1, then_2, then_3)
   n12 = OR Boolean!(A=Get bIsSkydiveFloating(), B=Get bIsCrouching(), C=Get bIsSwimSprinting()) -> (ReturnValue)
   > Set Constraints Alpha(Constraints Alpha:float=float * float(A="To Float (Boolean)"(InBool=@n12), B=Get Look up Alpha() -> (Look up Alpha:double)) {is=title K2Node_PromotableOperator}) -> (Output_Get:float)
   n13 = OR Boolean!(A=Get bIsFalling(), B=Get bIsJumping()) -> (ReturnValue)
   > Set Hood jump Falling(Hood jump Falling:float="To Float (Boolean)"(InBool=@n13)) -> (Output_Get:float)
   n14 = AND Boolean(A=Get bIsSprinting(), B=NOT Boolean(A=@n13))
   > Set Hood Sprint(Hood Sprint:float="To Float (Boolean)"(InBool=@n14)) -> (Output_Get:float)
   n15 = NOT Boolean(A=Get is wearing Backbling() -> (is wearing Backbling:bool))
   n16 = Get bIsSkydiveFloating()
   > Set hood jump falling No BP(hood jump falling No BP:float="To Float (Boolean)"(InBool=AND Boolean!(A=@n14, B=@n15, C:bool=@n16) -> (ReturnValue))) -> (Output_Get:float)
   n17 = Branch(execute=@n11.then_1, Condition=NOT Boolean(A=Get bIsPlayingMeleeAnim()))
   > Set Not playing Malee Anim(Not playing Malee Anim:float=Get Not Playing Malee Positive() -> (Not Playing Malee Positive:float)) -> (Output_Get:float)
   Set Not playing Malee Anim(execute=@n17.else, Not playing Malee Anim=Get Not Playing Malee Neg() -> (Not Playing Malee Neg:float))
   n18 = Get bIsJumping()
   n19 = Get bIsFalling()
   n20 = Branch(execute=@n11.then_3, Condition=AND Boolean(A=Get bIsPlayingMeleeAnim(), B=OR Boolean(A=@n18, B=@n19), C=Get is wearing Backbling()))
   > Set falling playing melee(falling playing melee:float=Get positive() -> (positive:float)) -> (Output_Get:float)
   Set falling playing melee(execute=@n20.else, falling playing melee=Get negative() -> (negative:float))
   n21 = Get bIsSprinting()
   n22 = Get bIsJumping()
   n23 = Get bIsFalling()
   Set is Falling Jumping Sprinting(
    execute=@n11.then_2
    is Falling Jumping Sprinting:float="To Float (Boolean)"(InBool=OR Boolean!(A=@n23, B=@n22, C:bool=@n21) -> (ReturnValue))
   ) -> (Output_Get:float)
   n24 = Get bIsSkydiveFloating()
   > Set Skydive Floating(Skydive Floating:float="To Float (Boolean)"(InBool=@n24)) -> (Output_Get:float)
   n25 = NOT Boolean(A=Get is wearing Backbling())
   > Set Skydive floating without BP(Skydive floating without BP:float="To Float (Boolean)"(InBool=AND Boolean!(A=@n24, B=@n25) -> (ReturnValue))) -> (Output_Get:float)
   n26 = Get bIsSkydiveFloating()
   n27 = Get bIsParachuteOpened()
   n28 = Get bIsSkyDiving()
   n29 = AND Boolean(A=Get bIsFalling(), B=NOT Boolean(A=OR Boolean(A=@n26, B=@n28, C=@n27)))
   > Set Falling(Falling:float="To Float (Boolean)"(InBool=@n29)) -> (Output_Get:float)
   n30 = NOT Boolean(A=Get is wearing Backbling())
   > Set Falling without BP(Falling without BP:float="To Float (Boolean)"(InBool=AND Boolean(A=@n29, B=@n30))) -> (Output_Get:float)
   > Set Parachute opened(Parachute opened:float="To Float (Boolean)"(InBool=Get bIsParachuteOpened())) -> (Output_Get:float)
   > n31 = Branch(Condition=Get is wearing Backbling())
   n32 = Get positive()
   > "Set Wearing Backbling "("Wearing Backbling ":float=@n32) -> (Output_Get:float)
   n33 = Get negative()
   > Set Not Wearing Backbling(Not Wearing Backbling:float=@n33) -> (Output_Get:float)
   "Set Wearing Backbling "(execute=@n31.else, "Wearing Backbling "=@n33)
   > Set Not Wearing Backbling(Not Wearing Backbling=@n32)
   Event BlueprintInitializeAnimation()
   > Set Skel Mesh(Skel Mesh=GetOwnerSkeletalMeshComponent())
   Event OnCharacterCustomizationLoaded()
   > Do Once()
   n34 = CheckForBackbling()
   > Set is wearing Backbling(is wearing Backbling:bool=OR Boolean!(A=@n34, B=@n34.bIsBackblingACape) -> (ReturnValue)) -> (Output_Get:bool)
   > n35 = Cast To PlayerPawn_Athena(Object=TryGetPawnOwner())
   > Set Body_Mesh(Body_Mesh=GetSkeletalMeshForPartType(self=@n35))
  ]}
]
variables [
 Body_Mesh {type=SkeletalMeshComponent}
 Clavicle_L_rot {type=number}
 Clavicle_R_rot {type=number}
 Constraints Alpha {type=number, default=1}
 Falling {type=number, default=1}
 falling playing melee {type=number, default=1}
 Falling without BP {type=number, default=1}
 Gravity Override Params {type=GravityOverrideParamsStruct, default {skydive_DivingDown {windAmplitude=<100 600 200>}, skydive_FloatingDown {windFrequency=<3 15 16>}}}
 GravityOverride {type=Vector}
 "head : Rotation" {type=number, default=0.25}
 "Head : Rotation Crouching" {type=number, default=0.4}
 "Head : Rotation_skydive" {type=number, default=0.5}
 Hood jump Falling {type=number, default=1}
 hood jump falling No BP {type=number}
 Hood Sprint {type=number, default=1}
 Hood Top 2 BP Rotation {type=Rotator, default=<rot 20 0 0>}
 Hood Top 2 BP Translation {type=Vector, default=<0.5 0 0>}
 Hood Top 2 No BP Rotation {type=Rotator, default=<rot 20 0 0>}
 Hood Top 2 No BP Translation {type=Vector}
 Hood Top BP Rotation {type=Rotator, default=<rot 20 0 0>}
 Hood Top BP Translation {type=Vector, default=<1.5 0 0>}
 Hood Top No BP Rotation {type=Rotator, default=<rot 15 0 0>}
 Hood Top No BP Translation {type=Vector, default=<10 -0.5 0>}
 Hood Top Translation {type=Vector}
 HoodTop Rotation {type=Rotator, default=<rot 20 0 0>}
 HoodTop2 Rotation {type=Rotator, default=<rot 20 0 0>}
 HoodTop2 Translation {type=Vector}
 is Falling Jumping Sprinting {type=number, default=0.5}
 Is Jumping Falling {type=number}
 is wearing Backbling {type=boolean}
 Look up Alpha {type=number}
 Look Up Alpha HAIR {type=number}
 Look Up Alpha_false {type=number}
 Look Up Alpha_true {type=number, default=1}
 negative {type=number}
 Not playing Malee Anim {type=number, default=1}
 Not Playing Malee Neg {type=number}
 Not Playing Malee Positive {type=number, default=1}
 Not Wearing Backbling {type=number, default=1}
 Parachute opened {type=number, default=1}
 positive {type=number, default=1}
 Skel Mesh {type=SkeletalMeshComponent}
 Skydive Floating {type=number, default=1}
 Skydive floating without BP {type=number, default=1}
 "Spine 05 : Rotation Crouching" {type=number, default=0.6}
 "Spine 05 : Rotation Skydive" {type=number, default=0.5}
 "spine_05 : Rotation" {type=number, default=0.75}
 "Wearing Backbling " {type=number, default=1}
]
